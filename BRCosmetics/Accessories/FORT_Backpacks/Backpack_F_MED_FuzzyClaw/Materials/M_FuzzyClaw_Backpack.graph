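# Character material for the FuzzyClaw backpack that shows timed eye expressions, labeled blink and elim, on top of the 2023 character parent PreFX/PostFX functions.
# Red vertex-color masks for default, star and closed eyes collapse unwanted meshes toward object position via WPO, with PreviousFrameSwitch zeroing that offset's previous frame.
# Blueprint-fed time, duration and expression parameters drive the switch and an emissive flash.
# A masked sine bob adds motion.
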Material /BRCosmetics/Accessories/FORT_Backpacks/Backpack_F_MED_FuzzyClaw/Materials/M_FuzzyClaw_Backpack
graph {
 properties {
  bUseMaterialAttributes=true
  bUsedWithMeshDeformer=true
  bUsedWithSkeletalMesh=true
  pixelDepthOffsetMode=PDOM_Legacy
  shadingModel=MSM_FromMaterialExpression
  usedShadingModels=Default Lit | Subsurface | Clear Coat | Subsurface Profile | Cloth
 }
 nodes [
  Frac(None=Multiply(A=Time(), B=Constant() {r=0.3}))
  n1 = MaterialFunctionCall() {materialFunction=&MF_CharacterParent_2023_PreFX}
  n2 = GetMaterialAttributes(None=@n1.Result) {attributeGetTypes [#1, #2]}
  MaterialFunctionCall(
   Material=SetMaterialAttributes!(MaterialAttributes=@n2.MaterialAttributes, World Position Offset=NamedRerouteUsage(), Emissive Color=NamedRerouteUsage(), Base Color) {attributeSetTypes [#3, #1, #2]}
   M=@n1.M
   Pre Skinned Local Position=@n1.M
   Distance Blend=@n1.M
   Fresnel=@n1.M
  ) {materialFunction=&MF_CharacterParent_2023_PostFX}
  n3 = VertexColor()
  n4 = MaterialFunctionCall!(Hide Element Toggle=Constant() {r=1}, VertexColorChannel=@n3.R, Vertex Value to hide=Constant() {r=0.1}) {desc=Default, materialFunction=/Game/Packages/Fortress_SharedMaterials/Base_Material_Functions/VertexColorMask}
  n5 = MaterialFunctionCall(Hide Element Toggle=Constant() {r=1}, VertexColorChannel=@n3.R, Vertex Value to hide=Constant() {r=0.2}) {desc=Star, materialFunction=&VertexColorMask}
  n6 = MaterialFunctionCall(Hide Element Toggle=Constant() {r=1}, VertexColorChannel=@n3.R, Vertex Value to hide=Constant() {r=0.3}) {desc=Eye closed, materialFunction=&VertexColorMask}
  n7 = Max(A=Max(A=@n4.Result, B=NamedRerouteDeclaration(None=@n5.Result) {name=Stareyes, nodeColor=<rgba 0.7109 1 0 1>}), B=@n6.Result)
  n8 = Subtract(A=@n7, B=@n4.Result)
  n9 = Subtract(A=@n7, B=@n6.Result)
  n10 = Saturate(None=ScalarParameter() {defaultValue=1, group=FuzzyClaw, parameterName=FC_ExpressionInt})
  n11 = Constant() {desc=Threshold, r=0.5}
  n12 = Constant()
  n13 = Subtract(
   A=Subtract(A=Time(), B=ScalarParameter() {group=FuzzyClaw, parameterName=FC_GameTimeFromBP})
   B=ScalarParameter() {defaultValue=2.5, group=FuzzyClaw, parameterName=FC_EmpressionDuration}
  )
  n14 = If(A=@n13, B=@n11, "A > B"=@n12, "A == B"=@n12, "A < B"=Constant() {r=1}) {constB=1}
  n15 = Multiply(A=OneMinus(None=@n10), B=@n14)
  n16 = Multiply(
   A=LinearInterpolate(A=LinearInterpolate(A=@n8, B=@n9, Alpha=@n15) {desc=Blink}, B=@n6.Result, Alpha=Multiply(A=@n10, B=@n14)) {desc=Elim}
   B=Subtract(A=ObjectPositionWS(), B=WorldPosition().XYZ)
  )
  n17 = Constant()
  n18 = Constant()
  n19 = Sine(None=Multiply(A=Time()) {constB=0.85})
  n20 = Transform(None=Multiply(A=MaterialFunctionCall(X=@n18, Y=@n18, Z=@n19) {materialFunction=&MakeFloat3}.Result) {constB=0.625}) {transformSourceType=TRANSFORMSOURCE_Local}
  n21 = NamedRerouteUsage()
  NamedRerouteDeclaration(None=Add(A=PreviousFrameSwitch(Current Frame=@n16, Previous Frame=@n17), B=LinearInterpolate(B=@n20, Alpha=@n21))) {name=WPO_OUT, nodeColor=<rgba 0 0.6953 1 1>}
  NamedRerouteDeclaration(None=@n2.Emissive Color) {name=Emissive_IN, nodeColor=<rgba 0.5938 0 1 1>}
  n22 = NamedRerouteUsage()
  n23 = LocalPosition() {localOrigin=InstancePreSkinning}
  n24 = Step(X=ComponentMask(None=VertexInterpolator(VS=@n23.XYZ).PS) {b=true}) {constX=130, constY=-5}
  n25 = Multiply(A=NamedRerouteUsage(), B=Constant3Vector() {constant=<rgba 0.151 0.0688 0 1>})
  n26 = NamedRerouteUsage()
  n27 = Distance(A=Saturate(None=@n13), B=Divide(A=@n11))
  n28 = OneMinus(None=Saturate(None=Multiply(A=@n27) {constB=2}))
  NamedRerouteDeclaration(None=LinearInterpolate(A=Multiply(A=@n22, B=@n24), B=Add(A=@n25, B=@n26), Alpha=Multiply(A=@n10, B=@n28)) {constAlpha=0}) {name=Emissive_OUT, nodeColor=<rgba 0 1 0.6719 1>}
  NamedRerouteDeclaration(None=@n2.Base Color) {name=Basecolor_IN, nodeColor=<rgba 0.3125 1 0 1>}
 ]
}
