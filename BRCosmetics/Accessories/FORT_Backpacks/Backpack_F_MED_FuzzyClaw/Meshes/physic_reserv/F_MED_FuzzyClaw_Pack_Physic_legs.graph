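# Physics asset for the legs of the FuzzyClaw backpack cosmetic, giving its body and two legs capsule bodies for physics-driven motion.
# Each side chains the body to a small base-leg capsule with free swing and twist, then to a larger dynamic leg capsule with tight swing and wider twist limits.
# The sides are asymmetric: the left leg has a fatter capsule and looser limits than the right.
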PhysicsAsset /BRCosmetics/Accessories/FORT_Backpacks/Backpack_F_MED_FuzzyClaw/Meshes/physic_reserv/F_MED_FuzzyClaw_Pack_Physic_legs
bodies {
 base_leg_l [{shapeType=Capsule, length=0.505, radius=0.505, shapeName=base_leg_l_capsule, rotation=<rot 0 -90 0>}]
 base_leg_r [{shapeType=Capsule, length=0.505, radius=0.505, shapeName=base_leg_r_capsule, rotation=<rot 0 -90 0>}]
 dyn_leg_l [{shapeType=Capsule, length=2, radius=3.5, shapeName=dyn_leg_l_capsule, center=<0.0142 -0.0425 0.6684>, rotation=<rot 0 -161.5586 3.8343>}]
 dyn_leg_r [{shapeType=Capsule, length=2, radius=2.8779, shapeName=dyn_leg_r_capsule, center=<-0.0108 -0.0257 1.2069>, rotation=<rot 0 157.2673 1.3208>}]
 dyn_uappee_Body [{
   shapeType=Capsule
   length=1.9718
   radius=6.1465
   shapeName=dyn_uappee_Body_capsule
   center=<0.0306 1.5603 -1.1648>
   rotation=<rot 0 90.6477 -89.1947>
  }]
}
constraints [
 {
  bone1Name=base_leg_l
  bone2Name=dyn_uappee_Body
  swing1LimitDegrees=10
  swing1Motion=Free
  swing2LimitDegrees=10
  swing2Motion=Free
  twistLimitDegrees=10
  twistMotion=Free
 }
 {bone1Name=dyn_leg_l, bone2Name=base_leg_l, swing1LimitDegrees=10, swing2LimitDegrees=10, twistLimitDegrees=45}
 {
  bone1Name=base_leg_r
  bone2Name=dyn_uappee_Body
  swing1LimitDegrees=10
  swing1Motion=Free
  swing2LimitDegrees=10
  swing2Motion=Free
  twistLimitDegrees=10
  twistMotion=Free
 }
 {bone1Name=dyn_leg_r, bone2Name=base_leg_r, swing1LimitDegrees=8, swing2LimitDegrees=8, twistLimitDegrees=32}
]
