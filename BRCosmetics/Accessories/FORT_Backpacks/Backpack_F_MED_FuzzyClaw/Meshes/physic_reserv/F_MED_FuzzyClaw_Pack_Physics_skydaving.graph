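# Physics asset for the FuzzyClaw backpack cosmetic, apparently a skydiving variant by its name, giving its head and tail limited secondary motion.
# Capsule bodies link the head and a rotation-locked tail base to the body.
# Three tail segments chain from that base, with tight twist and swing limits that widen toward the tip.
# A large pelvis capsule has no constraint, likely serving as a collision volume for the wearer.
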PhysicsAsset /BRCosmetics/Accessories/FORT_Backpacks/Backpack_F_MED_FuzzyClaw/Meshes/physic_reserv/F_MED_FuzzyClaw_Pack_Physics_skydaving
bodies {
 Tail_base [{shapeType=Capsule, length=0.505, radius=0.505, shapeName=Tail_base_capsule, rotation=<rot 0 -90 0>}]
 dyn_head [{shapeType=Capsule, length=0, radius=10.2887, shapeName=dyn_head_capsule, center=<0 2.7232 8>, rotation=<rot 0 0 10>}]
 dyn_tail_01 [{shapeType=Capsule, length=3.8042, radius=3.3498, shapeName=dyn_tail_01_capsule, rotation=<rot -0.0712 173.5265 -89.2782>}]
 dyn_tail_02 [{shapeType=Capsule, length=4.6233, radius=3.3507, shapeName=dyn_tail_02_capsule, rotation=<rot -0.6562 -3.6522 -83.3207>}]
 dyn_tail_03 [{shapeType=Capsule, length=4.0086, radius=2.7299, shapeName=dyn_tail_03_capsule, rotation=<rot -0.1515 179.5636 -88.4625>}]
 dyn_uappee_Body [{shapeType=Capsule, length=3.6796, radius=5.2926, center=<0 -2.7003 -2.6218>, rotation=<rot 90 180 179.9997>}]
 pelvis_1 [{shapeType=Capsule, length=29.217, radius=16.6564, shapeName=pelvis_1_capsule, center=<0.5 5.2 0>, rotation=<rot 0 -90 90>}]
}
constraints [
 {bone1Name=dyn_head, bone2Name=dyn_uappee_Body, swing1LimitDegrees=1, swing2LimitDegrees=18, twistLimitDegrees=16}
 {
  bone1Name=Tail_base
  bone2Name=dyn_uappee_Body
  swing1LimitDegrees=45
  swing1Motion=Locked
  swing2LimitDegrees=45
  swing2Motion=Locked
  twistLimitDegrees=45
  twistMotion=Locked
 }
 {bone1Name=dyn_tail_01, bone2Name=Tail_base, swing1LimitDegrees=9, swing2LimitDegrees=12, twistLimitDegrees=1}
 {bone1Name=dyn_tail_02, bone2Name=dyn_tail_01, swing1LimitDegrees=10, swing2LimitDegrees=15, twistLimitDegrees=1}
 {bone1Name=dyn_tail_03, bone2Name=dyn_tail_02, swing1LimitDegrees=11, swing2LimitDegrees=20, twistLimitDegrees=1}
]
