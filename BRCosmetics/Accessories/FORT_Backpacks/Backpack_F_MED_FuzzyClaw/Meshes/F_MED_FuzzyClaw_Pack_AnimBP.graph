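# Animation Blueprint for the FuzzyClaw cosmetic backpack, posing and physically animating its pack, tail and fenders on the character.
# The AnimGraph chains pack and tail bone offsets, pins pelvis_1 to the parent mesh's world-space pelvis, then runs RigidBody and fender AnimDynamics simulations.
# The EventGraph derives FuzzyClaw meta-tag and skydive flags, gravity, back-arch alphas and spine-driven offsets.
# Both graphs retain many disconnected experimental nodes.
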AnimBlueprint /BRCosmetics/Accessories/FORT_Backpacks/Backpack_F_MED_FuzzyClaw/Meshes/F_MED_FuzzyClaw_Pack_AnimBP
target_skeleton=&/F_MED_FuzzyClaw_Pack_Skeleton
graphs [
 AnimGraph {nodes [
   n1 = Anim Dynamics - Bone: fender_r!(ComponentPose, Alpha=1, GravityOverride:Vector="0,0,0", ExternalForce="0,0,0") -> (Pose) {
    angularDampingOverride=0.999
    angularSpringConstant=3333
    bAngularSpring=true
    bOverrideAngularDamping=true
    bUseGravityOverride=true
    gravityScale=0.9
    simulationSpace=RootRelative
    alphaBoolBlend {blendOption=Linear}
    boundBone {boneName=fender_r}
    componentAppliedLinearAccClamp=<3000 3000 3000>
    componentLinearAccScale=<2 2 2>
    componentLinearVelScale=<2 2 2>
    physicsBodyDefinitions [{
      boundBone {boneName=fender_r}
      boxExtents=<8 8 8>
      constraintSetup {angularLimitsMax=<10 5 10>, angularLimitsMin=<-33 -1 -10>, angularTarget=<0 0 0>}
      localJointOffset=<0.5 0 0>
     }]
   }
   Anim Dynamics - Bone: fender_l!(ComponentPose=@n1, Alpha=1, GravityOverride:Vector="0,0,0", ExternalForce="0,0,0") -> (Pose) {
    angularDampingOverride=0.999
    angularSpringConstant=3333
    bAngularSpring=true
    bOverrideAngularDamping=true
    bUseGravityOverride=true
    gravityScale=0.9
    simulationSpace=RootRelative
    alphaBoolBlend {blendOption=Linear}
    boundBone {boneName=fender_l}
    componentAppliedLinearAccClamp=<3000 3000 3000>
    componentLinearAccScale=<2 2 2>
    componentLinearVelScale=<2 2 2>
    physicsBodyDefinitions [{
      boundBone {boneName=fender_l}
      boxExtents=<8 8 8>
      constraintSetup {angularLimitsMax=<10 5 10>, angularLimitsMin=<-33 -1 -10>, angularTarget=<0 0 0>}
      localJointOffset=<0.5 0 0>
     }]
   }
   "Transform (Modify) Bone - Bone: pack"(Translation="0,0,1", Rotation="0,0,11") {
    rotationMode=BMM_Additive
    rotationSpace=BCS_BoneSpace
    translationMode=BMM_Replace
    alphaBoolBlend {blendOption=Linear}
    boneToModify {boneName=pack}
    rotation {roll=11}
    translation {z=1}
   }
   n2 = "Transform (Modify) Bone - Bone: pack"(Translation="0,-0.5,2", Rotation="0,0,11") {
    rotationMode=BMM_Additive
    rotationSpace=BCS_BoneSpace
    translationMode=BMM_Replace
    alphaBoolBlend {blendOption=Linear}
    boneToModify {boneName=pack}
    rotation {roll=11}
    translation {y=-0.5, z=2}
   }
   "Transform (Modify) Bone - Bone: pack"(ComponentPose=@n2, Translation="0,1,1", Rotation="0,0,-4") {
    rotationMode=BMM_Additive
    rotationSpace=BCS_BoneSpace
    translationMode=BMM_Replace
    alphaBoolBlend {blendOption=Linear}
    boneToModify {boneName=pack}
    rotation {roll=-4}
    translation {y=1, z=1}
   }
   "Transform (Modify) Bone - Bone: pack"!(ComponentPose, bAlphaBoolEnabled:bool, Translation="0,0,0", Rotation="0,0,0", Scale="0.9,0.9,0.9") -> (Pose) {
    alphaInputType=Bool
    rotationMode=BMM_Additive
    rotationSpace=BCS_BoneSpace
    translationMode=BMM_Additive
    alphaBoolBlend {blendOption=Linear}
    boneToModify {boneName=pack}
    scale=<0.9 0.9 0.9>
   }
   "Transform (Modify) Bone - Bone: Tail_base"(Alpha=1, Translation="0,0,0", Rotation="0,0,0", Scale="1,1,1") {rotationMode=BMM_Additive, rotationSpace=BCS_BoneSpace, alphaBoolBlend {blendOption=Linear}, boneToModify {boneName=Tail_base}}
   "Transform (Modify) Bone - Bone: Tail_base"(Alpha=1, Translation="0,0,0", Rotation="0,0,0", Scale="1,1,1") {
    rotationMode=BMM_Additive
    rotationSpace=BCS_BoneSpace
    translationMode=BMM_Additive
    alphaBoolBlend {blendOption=Linear}
    boneToModify {boneName=Tail_base}
   }
   n3 = "Transform (Modify) Bone - Bone: pack"!(ComponentPose, Alpha:float, Translation="0,0,0", Rotation="0,0,0", Scale) -> (Pose) {
    rotationMode=BMM_Additive
    rotationSpace=BCS_BoneSpace
    translationMode=BMM_Replace
    alphaBoolBlend {blendOption=Linear}
    boneToModify {boneName=pack}
   }
   n4 = "Transform (Modify) Bone - Bone: pack"!(ComponentPose=@n3, bAlphaBoolEnabled:bool, Translation="0,-1,1", Rotation="0,0,10", Scale) -> (Pose) {
    alphaInputType=Bool
    rotationMode=BMM_Additive
    rotationSpace=BCS_BoneSpace
    translationMode=BMM_Additive
    translationSpace=BCS_BoneSpace
    alphaBoolBlend {blendOption=Linear}
    boneToModify {boneName=pack}
    rotation {roll=10}
    translation {y=-1, z=1}
   }
   n5 = "Transform (Modify) Bone - Bone: Tail_base"!(ComponentPose=@n4, bAlphaBoolEnabled:bool=True, Translation="0,0,0", Rotation="30,0,0", Scale="1,1,1") -> (Pose) {
    alphaInputType=Bool
    rotationMode=BMM_Additive
    rotationSpace=BCS_BoneSpace
    translationMode=BMM_Additive
    translationSpace=BCS_BoneSpace
    alphaBoolBlend {blendOption=Linear}
    boneToModify {boneName=Tail_base}
    rotation {pitch=30}
   }
   "Transform (Modify) Bone - Bone: pack"(ComponentPose=@n5, Translation="0,-2,3", Rotation="0,0,26.6") {
    alphaInputType=Bool
    rotationMode=BMM_Additive
    rotationSpace=BCS_BoneSpace
    translationMode=BMM_Additive
    translationSpace=BCS_BoneSpace
    alphaBoolBlend {blendOption=Linear}
    boneToModify {boneName=pack}
    rotation {roll=26.6}
    translation {y=-2, z=3}
   }
   "Transform (Modify) Bone - Bone: pack"(Translation="0,0,0", Rotation="0,0,0", Scale="0.9,0.9,0.9") {
    alphaInputType=Bool
    rotationMode=BMM_Additive
    rotationSpace=BCS_BoneSpace
    translationMode=BMM_Additive
    alphaBoolBlend {blendOption=Linear}
    boneToModify {boneName=pack}
    scale=<0.9 0.9 0.9>
   }
   "Transform (Modify) Bone - Bone: pack"(Translation="0,0,0", Rotation="0,0,3") {
    alphaInputType=Bool
    rotationMode=BMM_Additive
    rotationSpace=BCS_BoneSpace
    translationMode=BMM_Additive
    translationSpace=BCS_BoneSpace
    alphaBoolBlend {blendOption=Linear}
    boneToModify {boneName=pack}
    rotation {roll=3}
   }
   RigidBody!(ComponentPose, Alpha, OverrideWorldGravity:Vector="0,0,-980", ExternalForce) -> (Pose) {
    bOverrideWorldGravity=true
    overridePhysicsAsset=/BRCosmetics/Accessories/FORT_Backpacks/Backpack_F_MED_FuzzyClaw/Meshes/physic_reserv/F_MED_FuzzyClaw_Pack_Physic_legs
    simulationSpace=BaseBoneSpace
    alphaBoolBlend {blendOption=Linear}
    baseBoneRef {boneName=pack}
    componentAppliedLinearAccClamp=<3500 3500 3500>
    componentLinearAccScale=<1 1 0.6>
    componentLinearVelScale=<0.5 1 0.8>
    simSpaceSettings {velocityScaleZ=0.8, worldAlpha=0.4}
   }
   RigidBody(OverrideWorldGravity="0,0,-980") {
    bOverrideWorldGravity=true
    overridePhysicsAsset=&F_MED_FuzzyClaw_Pack_Physic_legs
    simulationSpace=BaseBoneSpace
    alphaBoolBlend {blendOption=Linear}
    baseBoneRef {boneName=pack}
    componentLinearAccScale {x=0.2, z=0.2}
    componentLinearVelScale {x=0.2, z=0.2}
    simSpaceSettings {
     maxAngularAcceleration=1000
     maxAngularVelocity=3
     maxLinearAcceleration=700
     maxLinearVelocity=500
     velocityScaleZ=0.4
     worldAlpha=0.6
     externalLinearDragV=<0.3 0.3 0.3>
     externalLinearVelocity=<0.3 0.3 0.3>
    }
   }
   RigidBody!(ComponentPose, bAlphaBoolEnabled:bool, OverrideWorldGravity="0,0,-980", ExternalForce) -> (Pose) {
    alphaInputType=Bool
    bClampLinearTranslationLimitToRefPose=true
    bOverrideWorldGravity=true
    overridePhysicsAsset=&F_MED_FuzzyClaw_Pack_Physic_legs
    simulationSpace=BaseBoneSpace
    alphaBoolBlend {blendOption=Linear}
    baseBoneRef {boneName=dyn_uappee_Body}
    componentAppliedLinearAccClamp=<3000 3000 3000>
    componentLinearAccScale=<0.4 0.3 0.3>
    componentLinearVelScale=<0.4 0.3 0.3>
    overrideWorldGravity {z=-980}
    simSpaceSettings {
     maxAngularAcceleration=1000
     maxAngularVelocity=3
     maxLinearAcceleration=250
     maxLinearVelocity=500
     velocityScaleZ=0.1
     worldAlpha=0.5
     externalLinearDragV {y=0.5}
    }
   }
   n6 = "Transform (Modify) Bone - Bone: pack"!(ComponentPose=Local To Component(LocalPose=Input Pose()), Alpha:float, Translation="0,1,1", Rotation="0,0,5", Scale) -> (Pose) {
    rotationMode=BMM_Additive
    rotationSpace=BCS_BoneSpace
    translationMode=BMM_Replace
    alphaBoolBlend {blendOption=Linear}
    boneToModify {boneName=pack}
    rotation {roll=5}
    translation {y=1, z=1}
   }
   n7 = "Transform (Modify) Bone - Bone: pack"(ComponentPose=@n6, Alpha=0.3, Translation="0,0,0", Rotation="0,0,0") {
    rotationMode=BMM_Additive
    scaleMode=BMM_Additive
    scaleSpace=BCS_BoneSpace
    translationMode=BMM_Additive
    alphaBoolBlend {blendOption=Linear}
    boneToModify {boneName=pack}
   }
   n8 = "Transform (Modify) Bone - Bone: pack"(ComponentPose=@n7, Translation="0,-5,5", Rotation="0,0,44", Scale="0.9,0.9,0.9") {
    rotationMode=BMM_Additive
    rotationSpace=BCS_ParentBoneSpace
    translationMode=BMM_Additive
    translationSpace=BCS_ParentBoneSpace
    alphaBoolBlend {blendInTime=0.2, blendOption=Sinusoidal, blendOutTime=0.2}
    boneToModify {boneName=pack}
    rotation {roll=44}
    scale=<0.9 0.9 0.9>
    translation {y=-5, z=5}
   }
   n9 = "Transform (Modify) Bone - Bone: Tail_base"(ComponentPose=@n8, bAlphaBoolEnabled=Get bIsSkyDiving(), Translation="0,0,0", Rotation="-30,0,0", Scale="1,1,1") {
    alphaInputType=Bool
    rotationMode=BMM_Additive
    rotationSpace=BCS_BoneSpace
    translationMode=BMM_Additive
    translationSpace=BCS_ParentBoneSpace
    alphaBoolBlend {blendInTime=0.2, blendOption=Sinusoidal, blendOutTime=0.2}
    boneToModify {boneName=Tail_base}
    rotation {pitch=-30}
   }
   n10 = "Transform (Modify) Bone - Bone: pack"(ComponentPose=@n9, Translation="0,1,0", Rotation="0,0,10") {
    rotationMode=BMM_Additive
    rotationSpace=BCS_ParentBoneSpace
    translationMode=BMM_Additive
    translationSpace=BCS_ParentBoneSpace
    alphaBoolBlend {blendInTime=0.2, blendOption=Sinusoidal, blendOutTime=0.2}
    boneToModify {boneName=pack}
    rotation {roll=10}
    translation {y=1}
   }
   n11 = "Transform (Modify) Bone - Bone: Tail_base"!(ComponentPose=@n10, Alpha:float=1, Translation="0,0,0", Rotation="0,0,18", Scale="1,1,1") -> (Pose) {
    rotationMode=BMM_Additive
    rotationSpace=BCS_ParentBoneSpace
    translationMode=BMM_Additive
    translationSpace=BCS_ParentBoneSpace
    alphaBoolBlend {blendInTime=0.2, blendOption=Sinusoidal, blendOutTime=0.2}
    boneToModify {boneName=Tail_base}
    rotation {roll=18}
   }
   n12 = "Transform (Modify) Bone - Bone: pack"!(ComponentPose=@n11, bAlphaBoolEnabled:bool, Translation="0,-1.3,-4", Rotation="0,0,0", Scale) -> (Pose) {
    alphaInputType=Bool
    rotationMode=BMM_Additive
    rotationSpace=BCS_ParentBoneSpace
    translationMode=BMM_Additive
    translationSpace=BCS_ParentBoneSpace
    alphaBoolBlend {blendInTime=0.2, blendOption=Sinusoidal, blendOutTime=0.2}
    boneToModify {boneName=pack}
    translation {y=-1.3, z=-4}
   }
   n13 = "Transform (Modify) Bone - Bone: pack"(ComponentPose=@n12, Translation="0,0,-5", Rotation="0,0,13") {
    alphaInputType=Bool
    rotationMode=BMM_Additive
    rotationSpace=BCS_ParentBoneSpace
    translationMode=BMM_Additive
    translationSpace=BCS_ParentBoneSpace
    alphaBoolBlend {blendInTime=0.2, blendOption=Sinusoidal, blendOutTime=0.2}
    boneToModify {boneName=pack}
    rotation {roll=13}
    translation {z=-5}
   }
   n14 = "Transform (Modify) Bone - Bone: pack"(ComponentPose=@n13, Translation="0,0,3", Rotation="0,0,-5") {
    alphaInputType=Bool
    rotationMode=BMM_Additive
    rotationSpace=BCS_ParentBoneSpace
    translationMode=BMM_Additive
    translationSpace=BCS_ParentBoneSpace
    alphaBoolBlend {blendInTime=0.2, blendOption=Sinusoidal, blendOutTime=0.2}
    boneToModify {boneName=pack}
    rotation {roll=-5}
    translation {z=3}
   }
   n15 = "Transform (Modify) Bone - Bone: pack"!(ComponentPose=@n14, Alpha:float, Translation="0,3,-16", Rotation="0,0,0", Scale) -> (Pose) {
    rotationMode=BMM_Additive
    rotationSpace=BCS_BoneSpace
    translationMode=BMM_Additive
    translationSpace=BCS_BoneSpace
    alphaBoolBlend {blendInTime=0.2, blendOption=Sinusoidal, blendOutTime=0.2}
    boneToModify {boneName=pack}
    translation {y=3, z=-16}
   }
   n16 = "Transform (Modify) Bone - Bone: pelvis_1"!(
    ComponentPose=@n15
    bAlphaBoolEnabled:bool=True
    Translation=Get PelvisLoc() -> (PelvisLoc:Vector)
    Rotation=Get PelvisRot() -> (PelvisRot:Rotator)
    Scale=Get PelvisScale() -> (PelvisScale:Vector)
   ) -> (Pose) {
    alphaInputType=Bool
    rotationMode=BMM_Replace
    rotationSpace=BCS_WorldSpace
    scaleMode=BMM_Replace
    scaleSpace=BCS_WorldSpace
    translationMode=BMM_Replace
    translationSpace=BCS_WorldSpace
    alphaBoolBlend {blendInTime=0.2, blendOption=Linear, blendOutTime=0.2}
    boneToModify {boneName=pelvis_1}
    rotation {roll=25}
   }
   n17 = RigidBody(ComponentPose=@n16, OverrideWorldGravity="0,0,-980") {
    alphaInputType=Bool
    bClampLinearTranslationLimitToRefPose=true
    bOverrideWorldGravity=true
    overridePhysicsAsset=/BRCosmetics/Accessories/FORT_Backpacks/Backpack_F_MED_FuzzyClaw/Meshes/physic_reserv/F_MED_FuzzyClaw_Pack_Physics
    simulationSpace=BaseBoneSpace
    alphaBoolBlend {blendOption=Linear}
    baseBoneRef {boneName=pack}
    componentAppliedLinearAccClamp=<3333 3333 3333>
    componentLinearAccScale=<0.4 0.1 0.3>
    componentLinearVelScale=<0.4 0.1 0.3>
    overrideWorldGravity {z=-980}
    simSpaceSettings {
     maxAngularAcceleration=1000
     maxAngularVelocity=3
     maxLinearAcceleration=700
     maxLinearVelocity=500
     velocityScaleZ=0.1
     worldAlpha=0.3
     externalLinearDragV {y=0.2}
    }
   }
   n18 = RigidBody!(ComponentPose=@n17, Alpha:float, OverrideWorldGravity="0,0,-980", ExternalForce) -> (Pose) {
    bOverrideWorldGravity=true
    overridePhysicsAsset=&F_MED_FuzzyClaw_Pack_Physic_legs
    simulationSpace=BaseBoneSpace
    alphaBoolBlend {blendOption=Linear}
    baseBoneRef {boneName=pack}
    componentAppliedLinearAccClamp=<3500 3500 3500>
    componentLinearAccScale=<0.9 0.9 0.6>
    componentLinearVelScale=<0.5 0.9 0.8>
    simSpaceSettings {velocityScaleZ=0.6, worldAlpha=0.4, externalLinearDragV {y=0.2}}
   }
   n19 = Anim Dynamics - Bone: fender_l(ComponentPose=@n18, Alpha=1, GravityOverride="0,0,-980", ExternalForce="0,0,0") {
    angularDampingOverride=1
    angularSpringConstant=600
    bAngularSpring=true
    bOverrideAngularDamping=true
    bUseGravityOverride=true
    gravityScale=0.9
    simulationSpace=RootRelative
    alphaBoolBlend {blendOption=Linear}
    boundBone {boneName=fender_l}
    componentLinearAccScale=<1 0.5 0.5>
    componentLinearVelScale=<1 0.5 0.5>
    physicsBodyDefinitions [{
      boundBone {boneName=fender_l}
      boxExtents=<8 8 8>
      constraintSetup {angularLimitsMax=<0 3 5>, angularLimitsMin=<0 -3 -15>, angularTarget=<1 0 0>}
      localJointOffset=<0 -0.6 0>
     }]
   }
   n20 = Anim Dynamics - Bone: fender_r(ComponentPose=@n19, Alpha=1, GravityOverride="0,0,-980", ExternalForce="0,0,0") {
    angularDampingOverride=1
    angularSpringConstant=600
    bAngularSpring=true
    bOverrideAngularDamping=true
    bUseGravityOverride=true
    gravityScale=0.9
    simulationSpace=RootRelative
    alphaBoolBlend {blendOption=Linear}
    boundBone {boneName=fender_r}
    componentLinearAccScale=<1 0.5 0.5>
    componentLinearVelScale=<1 0.5 0.5>
    physicsBodyDefinitions [{
      boundBone {boneName=fender_r}
      boxExtents=<8 8 8>
      constraintSetup {angularLimitsMax=<0 3 15>, angularLimitsMin=<0 -3 -5>, angularTarget=<1 0 0>}
      localJointOffset=<0 0.6 0>
     }]
   }
   n21 = RigidBody!(ComponentPose=@n20, bAlphaBoolEnabled:bool, OverrideWorldGravity="0,0,-980", ExternalForce) -> (Pose) {
    alphaInputType=Bool
    bClampLinearTranslationLimitToRefPose=true
    bOverrideWorldGravity=true
    overridePhysicsAsset=/BRCosmetics/Accessories/FORT_Backpacks/Backpack_F_MED_FuzzyClaw/Meshes/physic_reserv/F_MED_FuzzyClaw_Pack_Physics_skydaving
    simulationSpace=BaseBoneSpace
    alphaBoolBlend {blendOption=Linear}
    baseBoneRef {boneName=pack}
    componentAppliedLinearAccClamp=<3333 3333 3333>
    componentLinearAccScale=<0.4 0.1 0.3>
    componentLinearVelScale=<0.4 0.1 0.3>
    overrideWorldGravity {z=-980}
    simSpaceSettings {
     maxAngularAcceleration=1000
     maxAngularVelocity=3
     maxLinearAcceleration=700
     maxLinearVelocity=500
     velocityScaleZ=0.1
     worldAlpha=0.3
     externalLinearDragV {y=0.2}
    }
   }
   Output Pose(Result=Component To Local(ComponentPose=@n21)) {layerGroup=None}
   Get notCrouch() -> (notCrouch:bool)
   Get GravityOverride()
  ]}
 EventGraph {nodes [
   n1 = Event BlueprintUpdateAnimation()
   > n2 = Sequence!(execute) -> (then_0, then_1, then_2:Exec, then_3:Exec)
   n3 = Set is FuzzyClaw(
    execute=[@n2.then_0, @n20.then]
    is FuzzyClaw:bool=CharacterItemDefinitionHasAnyMetaTag(Tags=Make Array("[0]"=(TagName="Cosmetics.MetaTags.FuzzyClaw")))
   ) -> (Output_Get:bool)
   > Set FuzzyClaw in skydaving(FuzzyClaw in skydaving:bool=AND Boolean(A=@n3, B=Get bIsSkyDiving())) -> (Output_Get:bool)
   > Set FuzzyClaw_parashuting(FuzzyClaw_parashuting:bool=AND Boolean(A=Get is FuzzyClaw() -> (is FuzzyClaw:bool), B=Get bIsParachuteOpened())) -> (Output_Get:bool)
   n4 = Get is FuzzyClaw()
   n5 = Get bIsSkyDiving()
   > Set FuzzyClaw_no_skyDaiving(FuzzyClaw_no_skyDaiving:bool=NOT Boolean(A=AND Boolean(A=@n4, B=@n5))) -> (Output_Get:bool)
   n6 = Gravity Override State Function DEPRECATED(
    execute=@n2.then_1
    CharacterPartAnimInstance=Self-Reference()
    GameTimeInSeconds=GetGameTimeInSeconds()
    GravityOverrideParams=Get Gravity Override Params()
   )
   > Set GravityOverride(GravityOverride=@n6)
   > n7 = Branch(Condition="float > float"(A=Get PawnVelocityZ(), B=1200))
   > Set VelUp(VelUp:double=1) -> (Output_Get:double)
   Set VelUp(execute=@n7.else, VelUp=0)
   n8 = GetPartSkeletalMeshComponent()
   Is Valid(exec=@n2.then_2, InputObject=@n8)
   n9 = GetAttachParent!(self:Scene Component Object Reference=@n8) -> (ReturnValue:Scene Component Object Reference)
   > Is Valid(InputObject=@n9)
   n10 = GetSocketTransform!(self=@n9, InSocketName=pelvis, TransformSpace=RTS_World) -> (ReturnValue_Location:Vector, ReturnValue_Rotation:Rotator, ReturnValue_Scale:Vector)
   > Set PelvisLoc(PelvisLoc:Vector=@n10) -> (Output_Get:Vector)
   > Set PelvisRot(PelvisRot:Rotator=@n10.ReturnValue_Rotation) -> (Output_Get:Rotator)
   > Set PelvisScale(PelvisScale:Vector=@n10.ReturnValue_Scale) -> (Output_Get:Vector)
   > n11 = Branch(Condition=Get bIsSkyDiving())
   > Set Pack Alpha(Pack Alpha:double=0.3) -> (Output_Get:double)
   Set Pack Alpha(execute=@n11.else, Pack Alpha=1)
   n12 = Sequence!(execute=@n2.then_3) -> (then_0, then_1)
   > Set Delta Time X(Delta Time X=@n1.DeltaTimeX)
   > Is Valid(InputObject=Get Body() -> (Body:Skeletal Mesh Component Budgeted Object Reference))
   n13 = GetDeltaTransformFromRefPose(self=Get Body(), BoneName=neck_01, BaseName=pelvis)
   n14 = MapRangeClamped(Value=BreakRotator(InRot=BreakTransform(InTransform=@n13).Rotation).Yaw, InRangeA=-87, InRangeB=-95)
   n15 = float * float(A=Get Delta Time X(), B=10) {is=title K2Node_PromotableOperator}
   n16 = FInterpTo(Current=Get BackArchAlpha() -> (BackArchAlpha:double), Target=@n14, DeltaTime=@n15, InterpSpeed=1.5)
   > Set BackArchAlpha(BackArchAlpha:double=@n16) -> (Output_Get:double)
   > Set BackArchAlpha_Tail(BackArchAlpha_Tail:double=FInterpTo(Current=Get BackArchAlpha_Tail() -> (BackArchAlpha_Tail:double), Target=@n14, DeltaTime=@n15, InterpSpeed=1.5)) -> (Output_Get:double)
   > Branch(Condition=Get FuzzyClaw in skydaving() -> (FuzzyClaw in skydaving:bool))
   > Set FuzzyClaw_Alpha(FuzzyClaw_Alpha:double=@n16) -> (Output_Get:double)
   n17 = GetOwnerSkeletalMeshComponent()
   n18 = BreakRotator(InRot=BreakTransform(InTransform=GetDeltaTransformFromRefPose(self=@n17, BoneName=spine_05, BaseName=spine_01)).Rotation)
   Set BackpackRotate(
    execute=@n12.then_1
    BackpackRotate:Rotator=MakeRotator(Roll=MapRangeClamped(Value=@n18.Yaw, InRangeA=0, InRangeB=-30, OutRangeA=3, OutRangeB=25))
   ) -> (Output_Get:Rotator)
   > Set BackpackTranslation(BackpackTranslation:Vector=MakeVector(Y=MapRangeClamped(Value=@n18.Yaw, InRangeA=0, InRangeB=-30, OutRangeB=-8))) -> (Output_Get:Vector)
   n19 = Event OnCharacterCustomizationLoaded()
   > Set Backpack(Backpack:Skeletal Mesh Component Object Reference=@n19.CustomBaseMesh) -> (Output_Get:Skeletal Mesh Component Object Reference)
   > n20 = Set Body(Body:Skeletal Mesh Component Budgeted Object Reference=Get Array Item with EFort Custom Part Type(Array=@n19.CustomMeshParts)) -> (Output_Get:Skeletal Mesh Component Budgeted Object Reference)
   n21 = Branch(Condition=Get bIsDBNO())
   > n22 = Set M_alpha(M_alpha:double=-1.8) -> (Output_Get:double)
   n23 = Reroute Node(InputPin:double) -> (OutputPin:double)
   n24 = GetOwnerSkeletalMeshComponent()
   n25 = BreakRotator(InRot=BreakTransform(InTransform=GetDeltaTransformFromRefPose(self=@n24, BoneName=spine_05, BaseName=pelvis)).Rotation)
   n26 = FInterpTo(
    Current=Get!() -> (BackpackRotate_Roll:float, BackpackRotate_Pitch:float, BackpackRotate_Yaw:float)
    Target=float + float(
     A=MapRangeClamped(Value=@n25.Yaw, InRangeA=-8, InRangeB=-30, OutRangeB=15)
     B=MapRangeClamped(Value=@n25.Yaw, InRangeA=-30, InRangeB=-81, OutRangeB=50)
    )
    DeltaTime=@n23
    InterpSpeed=20
   )
   Set BackpackRotate(execute=[@n22.then, @n29.then], BackpackRotate=MakeRotator(Roll=@n26))
   n27 = float + float(A=MapRangeClamped(Value=@n25.Yaw, InRangeA=-8, InRangeB=-30, OutRangeB=-3), B=Get M_alpha() -> (M_alpha:double))
   n28 = Get!() -> (BackpackTranslation_X:double, BackpackTranslation_Y:double, BackpackTranslation_Z:double)
   > Set BackpackTranslation(BackpackTranslation=MakeVector(Y=FInterpTo(Current=@n28.BackpackTranslation_Y, Target=@n27, DeltaTime=@n23, InterpSpeed=10)))
   n29 = Set M_alpha(execute=@n21.else, M_alpha=0)
   n30 = Branch(Condition=Get NuSkyDiving() -> (NuSkyDiving:bool))
   > Set M_alpha(M_alpha=0)
   > Set VelUp(VelUp=0)
   n31 = Branch(execute=@n30.else, Condition=Get isM_Med() -> (isM_Med:bool))
   > Set M_alpha(M_alpha=1)
   > Set VelUp(VelUp=0)
   Set M_alpha(execute=@n31.else, M_alpha=0)
   > Set VelUp(VelUp=1)
   n32 = GetOwnerSkeletalMeshComponent()
   n33 = BreakRotator(InRot=BreakTransform(InTransform=GetDeltaTransformFromRefPose(self=@n32, BoneName=spine_05, BaseName=spine_01)).Rotation)
   Set BackpackRotate(BackpackRotate=MakeRotator(Roll=MapRangeClamped(Value=@n33.Yaw, InRangeA=0, InRangeB=-16, OutRangeB=20)))
   > Set BackpackRotate_Y(BackpackRotate_Y:Rotator=MakeRotator(Pitch=MapRangeClamped(Value=@n33, InRangeA=27, InRangeB=-35, OutRangeB=-33))) -> (Output_Get:Rotator)
   > Set BackpackTranslation(BackpackTranslation=MakeVector(Y=0.5, Z=MapRangeClamped(Value=@n33.Yaw, InRangeA=25, InRangeB=-30, OutRangeB=8)))
   > Set NuSkyDiving(NuSkyDiving:bool=OR Boolean!(A=Get bIsSkydiveDiveMode(), B=false, C:bool=Get bIsParachuteOpened()) -> (ReturnValue)) -> (Output_Get:bool)
   > n34 = Branch(Condition=Get NuSkyDiving())
   > Set Pack Alpha(Pack Alpha=0)
   Set Pack Alpha(execute=@n34.else, Pack Alpha=1)
   MapRangeClamped(Value=0, InRangeA=0, InRangeB=-10, OutRangeB=20)
   MapRangeClamped(Value=0, InRangeA=30, InRangeB=-30, OutRangeB=8)
   GetDeltaTransformFromRefPose(BoneName=spine_05, BaseName=pelvis)
   MapRangeClamped(Value=0, InRangeA=-25, InRangeB=-40)
   MapRangeClamped(Value=0, InRangeA=-87, InRangeB=-95)
   TryGetPawnOwner()
   MakeRotator(Pitch=@n26)
   MapRangeClamped(Value=@n14, InRangeA=-25, InRangeB=-40)
   OR Boolean!(A="float >= float"(A=Get PawnVelocityZ(), B=1800), B=Get bIsPlayingMeleeAnim()) -> (ReturnValue)
   AND Boolean(A=false, B=false)
   AND Boolean(A=false, B="float < float"(B=1800))
   "Equal (Enum)"(
    A:EFortCustomGender Enum=Get GenderPermitted!(self:Custom Character Part Object Reference) -> (GenderPermitted:EFortCustomGender Enum=Invalid)
    B:EFortCustomGender Enum=Male
   )
   "Equal (Enum)"(
    A:EFortCustomBodyType Enum=Get BodyTypesPermitted!(self:Custom Character Part Object Reference) -> (BodyTypesPermitted:EFortCustomBodyType Enum=NONE)
    B:EFortCustomBodyType Enum=Medium
   )
   "Get (a copy)"(Array:Array of Wildcards, Dimension 1=1) -> (Output:Wildcard)
   Get bIsCrouching()
   Get bIsSkydiveFloating()
   Set isM_Med(isM_Med:bool=false) -> (Output_Get:bool)
   Set FuzzyClaw_Alpha(FuzzyClaw_Alpha=0)
  ]}
]
variables [
 BackArchAlpha {type=number}
 BackArchAlpha_Tail {type=number}
 Backpack {type=SkeletalMeshComponent}
 Backpack Translate {type=Vector}
 BackpackRotate {type=Rotator}
 BackpackRotate_Y {type=Rotator}
 BackpackTranslation {type=Vector}
 Body {type=SkeletalMeshComponentBudgeted}
 Delta Time X {type=number}
 Enabled {type=boolean}
 FuzzyClaw in skydaving {type=boolean}
 FuzzyClaw_Alpha {type=number}
 FuzzyClaw_no_skyDaiving {type=boolean}
 FuzzyClaw_parashuting {type=boolean}
 Game Time in Seconds {type=number}
 Gravity Override Params {type=GravityOverrideParamsStruct, default {skydive_Parachuting {windAmplitude=<200 -300 400>}}}
 GravityOverride {type=Vector}
 is FuzzyClaw {type=boolean}
 isM_Med {type=boolean}
 isMelee {type=number}
 M_alpha {type=number}
 notCrouch {type=boolean}
 NuSkyDiving {type=boolean}
 Pack Alpha {type=number}
 PelvisLoc {type=Vector}
 PelvisRot {type=Rotator}
 PelvisScale {type=Vector}
 VelUp {type=number}
]
